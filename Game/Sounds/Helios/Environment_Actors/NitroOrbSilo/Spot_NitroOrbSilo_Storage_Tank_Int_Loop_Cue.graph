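# Sound cue for the NitroOrbSilo storage tank interior ambient loop, built from Artemis oil rig storage tank audio and attenuation.
# A mixer blends two modulated layers under a quality node whose last two slots are empty.
# One layer shapes the tank loop with a repeating envelope that fades in, holds, fades out and rests silent.
# The other waits a fixed delay, then loops randomly chosen tank stop sounds, with the delay outside the loop.
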SoundCue /Game/Sounds/Helios/Environment_Actors/NitroOrbSilo/Spot_NitroOrbSilo_Storage_Tank_Int_Loop_Cue
root=QualityLevel_0
nodes [
 {id=Delay_1, children [Looping_2], properties {delayMax=12, delayMin=12}}
 {
  id=Enveloper_0
  children [WavePlayer_10]
  properties {
   bLoop=true
   bLoopIndefinitely=true
   loopEnd=24
   pitchCurve {editorCurveData {keys [{time=0}]}}
   volumeCurve {editorCurveData {keys [
      {arriveTangent=0.6828, interpMode=RCIM_Cubic, leaveTangent=0.6828, tangentMode=RCTM_User, time=0, value=0}
      {arriveTangent=-0.0017, interpMode=RCIM_Cubic, leaveTangent=-0.0017, tangentMode=RCTM_User, time=2, value=1}
      {arriveTangent=-0.0017, interpMode=RCIM_Cubic, leaveTangent=-0.0017, tangentMode=RCTM_User, time=12, value=1}
      <at 14 0>
      <at 24 0>
     ]}}
  }
 }
 {id=Looping_2, children [Random_1]}
 {id=Mixer_0, children [Modulator_0,Modulator_7], properties {inputVolume [3,2]}}
 {id=Modulator_0, children [Delay_1]}
 {id=Modulator_7, children [Enveloper_0]}
 {id=QualityLevel_0, children [Mixer_0,Mixer_0,null,null]}
 {id=Random_1, children [WavePlayer_1,WavePlayer_8,WavePlayer_9], properties {weights [1,1,1]}}
 {id=WavePlayer_1, properties {soundWaveAssetPtr=/Game/Sounds/Artemis/Environment_Actors/OilRigStorageTank/Spot_OilRig_Storage_Tank_Stop_01}}
 {
  id=WavePlayer_10
  properties {bLooping=true, soundWaveAssetPtr=/Game/Sounds/Artemis/Environment_Actors/OilRigStorageTank/Spot_OilRig_Storage_Tank_Loop}
 }
 {id=WavePlayer_8, properties {soundWaveAssetPtr=/Game/Sounds/Artemis/Environment_Actors/OilRigStorageTank/Spot_OilRig_Storage_Tank_Stop_02}}
 {id=WavePlayer_9, properties {soundWaveAssetPtr=/Game/Sounds/Artemis/Environment_Actors/OilRigStorageTank/Spot_OilRig_Storage_Tank_Stop_04}}
]
properties {
 attenuationSettings=/Game/Sounds/Artemis/Environment_Actors/OilRigStorageTank/Spot_OilRig_Storage_Tank_Att
 bEnableBusSends=false
 bEnableSubmixSends=false
 bOverrideConcurrency=true
 maxDistance=650
 soundClassObject=&Fort_Ambient_Outdoors_Athena
 volumeMultiplier=0.251
 concurrencyOverrides {maxCount=2, voiceStealReleaseTime=0.2}
}
