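# Looping ambient spot sound for the NitroOrbSilo storage tank, layering a tube rattle with a louder flowing liquid metal loop.
# The mixed loops pass through random modulation and a looping envelope that slowly varies pitch and volume.
# A quality switch plays this only on its first two levels, behind a short random delay.
# It borrows oil rig tank attenuation and tightly caps concurrent voices.
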SoundCue /Game/Sounds/Helios/Environment_Actors/NitroOrbSilo/Spot_NitroOrbSilo_Storage_Tank_Rattles_Loop_Cue
root=Delay_0
nodes [
 {id=Delay_0, children [QualityLevel_0], properties {delayMax=0.5}}
 {
  id=Enveloper_1
  children [Modulator_2]
  properties {
   bLoopIndefinitely=true
   loopEnd=30
   pitchCurve {editorCurveData {keys [{interpMode=RCIM_Cubic, time=0}, <at 15 1.1 RCIM_Cubic>, {interpMode=RCIM_Cubic, time=30}]}}
   volumeCurve {editorCurveData {keys [<at 0 1 RCIM_Cubic>,<at 15 0.7 RCIM_Cubic>,<at 30 1 RCIM_Cubic>]}}
  }
 }
 {id=Mixer_1, children [WavePlayer_2,WavePlayer_0], properties {inputVolume [1,2]}}
 {id=Modulator_2, children [Mixer_1], properties {pitchMin=0.8}}
 {id=QualityLevel_0, children [Enveloper_1,Enveloper_1,null,null]}
 {
  id=WavePlayer_0
  properties {bLooping=true, soundWaveAssetPtr=/Game/Sounds/Helios/POI/NitroEngine/Spot/Spot_NitroSilo_Liquid_Metal_Flowing_Inside_Loop}
 }
 {
  id=WavePlayer_2
  properties {bLooping=true, soundWaveAssetPtr=/Game/Sounds/Helios/Environment_Actors/NitroChimneyStack/Spot_Nitro_Chimney_Tube_Rattle}
 }
]
properties {
 attenuationSettings=/Game/Sounds/Artemis/Environment_Actors/OilRigStorageTank/Spot_OilRig_Storage_Tank_Att
 bEnableBusSends=false
 bEnableSubmixSends=false
 bOverrideConcurrency=true
 maxDistance=650
 soundClassObject=&Fort_Ambient_Outdoors_Athena
 volumeMultiplier=0.15
 concurrencyOverrides {maxCount=2, voiceStealReleaseTime=0.2}
}
